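; Music clock provider for the Niagara music clock data interface that supplies the clock from the music event subsystem.
; GetMusicClock breaks the provider context to get the world, fetches that game instance's MusicEventSubsystem, and returns its highest-priority clock.
; The data interface, Niagara system and attach component in the context are unused.
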(fn GetMusicClock
  (params (Context :type Niagara%20Data%20Interface%20Music%20Clock%20Provider%20Context%20Structure%20%28by%20ref%29))
  (bind (_musicclockndi _niagarasystem _attachcomponent _world) (Utilities|Struct|BreakNiagaraDataInterfaceMusicClockProviderContext Context))
  (return (Audio|GetHighestPriorityClock (GameInstanceSubsystems|GetMusicEventSubsystem _world))))
